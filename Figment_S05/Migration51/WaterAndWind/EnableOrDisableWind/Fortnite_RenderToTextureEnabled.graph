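# Material function that selects between material inputs with and without render-to-texture (RTT) features, acting as a shared compile-time toggle.
# One StaticBool, left at its default false, drives three static switches for Vector3, Vector2 and Material Attributes outputs, so each passes its "WithOut RTT" input.
# A "Water Interaction Enabled" collection parameter from FortniteMaterialParameters is present but unconnected, so it does not affect the selection.
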MaterialFunction /Figment_S05/Migration51/WaterAndWind/EnableOrDisableWind/Fortnite_RenderToTextureEnabled
graph {nodes [
  CollectionParameter() {
   collection=/Figment_S05/Migration51/Packages/Fortress_SharedMaterials/GlobalMaterialParameters/FortniteMaterialParameters
   parameterName=Water Interaction Enabled
  }
  n1 = StaticBool()
  FunctionOutput(None=StaticSwitch(True=FunctionInput() {inputName=With RTT Features}, False=FunctionInput() {inputName=WithOut RTT Features, sortPriority=1}, Value=@n1)) {outputName=V3}
  FunctionOutput(None=StaticSwitch(
    True=FunctionInput() {inputName=With RTT Features Att, inputType=FunctionInput_MaterialAttributes, sortPriority=5}
    False=FunctionInput() {inputName=WithOut RTT Features Att, inputType=FunctionInput_MaterialAttributes, sortPriority=6}
    Value=@n1
   )) {outputName=Att, sortPriority=3}
  FunctionOutput(None=StaticSwitch(
    True=FunctionInput() {inputName=With RTT Features V2, inputType=FunctionInput_Vector2, sortPriority=2}
    False=FunctionInput() {inputName=WithOut RTT Features V2, inputType=FunctionInput_Vector2, sortPriority=4}
    Value=@n1
   )) {outputName=V2, sortPriority=1}
 ]}
